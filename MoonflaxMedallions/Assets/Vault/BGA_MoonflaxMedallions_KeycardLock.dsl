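; Keycard lock for a Moonflax Medallions vault, linking one boss's vault door to that boss's vehicle convoy.
; On BeginPlay the authority finds or adds the boss convoy manager component on the GameState and listens for boss spawns.
; When a boss with the matching tag spawns in a convoy, it changes the lock state once, opens the attached door's gate and unbinds.
; On unlock, it notifies the manager that this boss's door opened.
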(event EventBeginPlay
  (bind _returnvalue (Game|GetGameState))
  (bind _outputdelegate (EventDispatchers|CreateEvent :!event Custom|OnSpawnedBoss_Event))
  (|CallParentFunction)
  (Networking|SwitchHasAuthority
    (:Authority
      (bind _output_get (Variables|Default|SetBossVehicleManager (Actor|GetComponentbyClass _returnvalue "/MoonflaxBossVehicle/Gameplay/B_BossVehicle_SpawnStateConvoy_Manager.B_BossVehicle_SpawnStateConvoy_Manager_C")))
      (Utilities|IsValid _output_get
        (:Is%20Valid
          (Default|AssignOnSpawnedBoss (Variables|Default|GetBossVehicleManager) _outputdelegate))
        (:Is%20Not%20Valid
          (bind _returnvalue_2 (Game|AddComponentbyClass _returnvalue "/MoonflaxBossVehicle/Gameplay/B_BossVehicle_SpawnStateConvoy_Manager.B_BossVehicle_SpawnStateConvoy_Manager_C"))
          (bind _output_get_2 (Variables|Default|SetBossVehicleManager _returnvalue_2))
          (Default|AssignOnSpawnedBoss _output_get_2 _outputdelegate))))))

(event Custom|MatchingBossSpawnedOnConvoy)

(event Custom|OnSpawnedBoss_Event
  (params (BossTag :type Gameplay%20Tag%20Container%20Structure))
  ;[0 0.5534 1] When the Boss in question spawns in their convoy, we open the gate from the start
  (if (GameplayTags|HasTag BossTag (Variables|Default|GetBossKeycardLockTag) true)
    (Utilities|FlowControl|DoOnce false
      (:Completed
        (CallFunction|SetLock "NewEnumerator2")
        (bind _asbga_moonflax_medallions_vault_door (Utilities|Casting|CastToBGA_MoonflaxMedallions_VaultDoor (Variables|LockParent|GetAttachedLockActor)))
        (Variables|/MoonflaxMedallions/Assets/Vault/BGA_MoonflaxMedallions_VaultDoor.BGA_MoonflaxMedallions_VaultDoor_C|SetIsGateOpen true _asbga_moonflax_medallions_vault_door)
        (Default|UnbindEventfromOnSpawnedBoss (Actor|GetComponentbyClass (Game|GetGameState) "/MoonflaxBossVehicle/Gameplay/B_BossVehicle_SpawnStateConvoy_Manager.B_BossVehicle_SpawnStateConvoy_Manager_C") (delegate Custom|OnSpawnedBoss_Event))))))

(event EventOnUnlocked
  (|CallParentFunction)
  (Class|BBossVehicleSpawnStateConvoyManager|NotifyDoorOpened (Variables|Default|GetBossVehicleManager) (Variables|Default|GetBossKeycardLockTag)))

(fn ConstructionScript
  (|CallParentFunction))
