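; Moonflax Medallions vault door Blueprint, replacing the parent's opening timeline with its own gate-opening animation.
; On authority it multicasts opening effects and runs a timeline that lerps the door's Z scale from closed to open.
; Slam particles, sound, camera shake and force feedback fire at start and finish where cosmetics can run.
; IsGateOpen's RepNotify snaps scale for clients outside relevancy and destroys the bullet blocker, unless the gate is still opening.
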(event EventBeginPlay
  ;[1 1 1] Set up references
  (|CallParentFunction)
  (Variables|Default|SetGateClosedScale (Variables|/Script/Engine.SceneComponent|GetRelativeScale3D (Variables|Default|GetVaultDoor))))

(event EventOpeningTimeline
  ;[0 0.9238 1] Overrides Parent Event so we can call our own timeline
  (Networking|SwitchHasAuthority
    (:Authority
      (CallFunction|MulticastPlayDoorOpenEffects)
      (CallFunction|CallGateOpenTimeline)
      (CallFunction|PlayGateSlamFX "0, 0, 0"))))

(event Custom|CallGateOpenTimeline
  ;[1 0 0.7825] Open Gate via Timeline. After it's finished, set variables and call RepNotify to update the gate to other players that might have been out of relevancy
  (Variables|Default|SetIsGateOpening true)
  (bind (_direction _gateheight) (Timeline|Timeline PlayFromStart)
    (:Update
      (CallFunction|UpdateGateOpening _gateheight))
    (:Finished
      (Variables|Default|SetIsGateOpening false)
      (Variables|Default|SetIsGateOpen true)
      (Networking|CanExecuteCosmeticEvents
        (:True
          (CallFunction|PlayGateSlamFX (Variables|Default|GetVFXSpawn%20Offset)))))))

(event Custom|UpdateGateOpening
  (params (Lerp :type Float%20%28double-precision%29))
  ;[1 0.7666 0] Updates Scale with Lerp
  (Transformation|SetRelativeScale3D (Variables|Default|GetVaultDoor) (Math|Vector|MakeVector 1.0 1.0 (Math|Float|Lerp (.z (Variables|Default|GetGateClosedScale)) (.z (Variables|Default|GetGateOpenScale)) Lerp))))

(event Custom|PlayGateSlamFX
  (params (VFXSpawnOffset :type Vector))
  (bind _frame (Variables|Default|GetFrame))
  (bind _returnvalue (Transformation|GetWorldLocation _frame))
  (bind _self self)
  (bind _returnvalue_2 (Transformation|GetActorLocation _self))
  (bind _returnvalue_3 (+ _returnvalue VFXSpawnOffset))
  ;[0.2777 1 0] Play Slam VFX when gate closes + screen shake
  (Networking|CanExecuteCosmeticEvents
    (:True
      (Effects|Components|ParticleSystem|SpawnFXAtLocation "/MoonflaxMedallions/VFX/Vault/NS_MoonflaxMedallions_SecureDoorSlam.NS_MoonflaxMedallions_SecureDoorSlam" _returnvalue_3 (Transformation|GetWorldRotation _frame))
      ;;[0.3793 0 1] Camera shake and force feedback for gate slamming shut and SFX
      (Audio|PlaySoundatLocation (Variables|Default|GetDoorSlamSFX) _returnvalue_3) ; Audio
      (Camera|PlayWorldCameraShake "/Game/Effects/Fort_Effects/BluePrints/CameraShakes/B_Medium_Explosion_CameraShake.B_Medium_Explosion_CameraShake_C" _returnvalue_2 0.0 (Variables|Default|GetCameraShakeRadius))
      (ForceFeedback|SpawnForceFeedbackatLocation "/Game/Weapons/ForceFeedback/Explosions/FF_Generic_Explosion_Near.FF_Generic_Explosion_Near" _returnvalue_2))))

(fn OnRep_IsGateOpen
  (Networking|CanExecuteCosmeticEvents
    (:True
      (if (not (Variables|Default|GetIsGateOpening))
        ;[0.4219 0 1] Replicate gate status for remote clients out of relevancy when gate was closed
        (Transformation|SetRelativeScale3D (Variables|Default|GetVaultDoor) (Math|Vector|SelectVector (Variables|Default|GetGateOpenScale) (Variables|Default|GetGateClosedScale) (Variables|Default|GetIsGateOpen)))
        (Components|DestroyComponent (Variables|Default|GetBulletBlocker))))))

(fn ConstructionScript
  (|CallParentFunction))
